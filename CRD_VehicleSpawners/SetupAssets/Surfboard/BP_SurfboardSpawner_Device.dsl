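; Surfboard spawner device for Creative, a Blueprint child that adds edit-mode collision and effect behavior to its parent spawner.
; HandleEditModeCollision forces the hologram's visibility mesh components to a trigger-only collision profile with Visibility set to overlap, regardless of prior setting.
; UpdateEffects plays the post-spawn ready sound at the actor's location when cosmetic events can run and the device is enabled.
; The construction script only calls the parent.
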(fn HandleEditModeCollision
  ;[0.9235 1 0.7671 0.75] Set Collision on hologram to TriggerOnly regardless of prior setting
  (bind _visibilitycomponents (CallFunction|GetVisibilityComponents))
  (for _array_element _visibilitycomponents
    (Utilities|IsValid _array_element
      (:Is%20Valid
        (bind _asstatic_mesh_component (Utilities|Casting|CastToStaticMeshComponent _array_element))
        (Collision|SetCollisionProfileName _asstatic_mesh_component "FortTriggerOnlyInteractions" true :!class /Script/Engine.PrimitiveComponent)
        (Collision|SetCollisionResponsetoChannel _asstatic_mesh_component "ECC_Visibility" "ECR_Overlap")))))

(fn UpdateEffects
  ;[1 1 1] Play post-spawn sound
  (Networking|CanExecuteCosmeticEvents
    (:True
      (if (Class|CreativeEnabledComponent|IsEnabled (Variables|Default|GetEnabledComponent))
        (Audio|PlaySoundatLocation (Variables|Cosmetics|GetTrap%20Ready%20Sound) (Transformation|GetActorLocation))))))

(fn ConstructionScript
  (|CallParentFunction))
