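; Spawner device Blueprint for the Surfboard vehicle in Creative, extending a parent spawner device.
; On initialization it gives the hologram preview mesh an opaque material for in-editor preview, otherwise the translucent one.
; Edit-mode handling forces trigger-only collision with visibility overlap on the hologram meshes.
; It also exposes a wrap-support option and plays a ready sound outside the editor when enabled and cosmetic events can run.
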(event EventInitializeDevice
  (bind _staticmeshcomponent (Variables|Debug|GetStaticMeshComponent))
  (|CallParentFunction)
  ;[1 1 1] Adjust the Material if we are in editor for the hologram, as Unreal isn't a fan of Selecting Translucent Materials(Or Just press "T")
  (if (and (Editor|InEditor) (Variables|Default|GetUseEditorPreview))
    (Rendering|Material|SetMaterial _staticmeshcomponent 0 "/Game/Creative/Devices/Generic/Materials/M_CP_Device_Hologram_Vehicles_Opaque.M_CP_Device_Hologram_Vehicles_Opaque")
    (else
      (Rendering|Material|SetMaterial _staticmeshcomponent 0 "/Game/Creative/Devices/Generic/Materials/M_CP_Device_Hologram_Vehicles.M_CP_Device_Hologram_Vehicles"))))

(fn GetSupportsWraps
  (return (Variables|UserOptions|Basic|GetSupportWraps)))

(fn HandleEditorVisibility)

(fn HandleEditModeCollision
  ;[0.9235 1 0.7671 0.75] Set Collision on hologram to TriggerOnly regardless of prior setting
  (bind _visibilitycomponents (CallFunction|GetVisibilityComponents))
  (for _array_element _visibilitycomponents
    (Utilities|IsValid _array_element
      (:Is%20Valid
        (bind _asstatic_mesh_component (Utilities|Casting|CastToStaticMeshComponent _array_element))
        (Collision|SetCollisionProfileName _asstatic_mesh_component "FortTriggerOnlyInteractions" true :!class /Script/Engine.PrimitiveComponent)
        (Collision|SetCollisionResponsetoChannel _asstatic_mesh_component "ECC_Visibility" "ECR_Overlap")))))

(fn SetSupportsWraps
  (params (Supports%20Wraps :type Boolean))
  (Variables|UserOptions|Basic|SetSupportWraps Supports%20Wraps))

(fn UpdateEffects
  ;[1 1 1] Play post-spawn sound
  (Networking|CanExecuteCosmeticEvents
    (:True
      (if (and (Class|CreativeEnabledComponent|IsEnabled (Variables|Default|GetEnabledComponent)) (not (Editor|InEditor)))
        (Audio|PlaySoundatLocation (Variables|Cosmetics|GetTrap%20Ready%20Sound) (Transformation|GetActorLocation))))))

(fn ConstructionScript
  (|CallParentFunction))
